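# Niagara environment effect that sheds snow from trees as GPU sprite flakes, smoke puffs and colored glints.
# All three emitters snap spawns to scene depth via GBuffer sampling, filter by green hue, and cull near screen edges.
# A system curve pulses flake and smoke spawning, with wind pushing both.
# Fixed world-space regions mask flakes and smoke and confine glints, while time of day drives smoke color and flake size.
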NiagaraSystem /Game/Effects/Fort_Effects/Effects/Environment/Hera/Generic/NS_Hera_Snow_FromTrees
dependencies {
 usedDataInterfaces [
  &NiagaraDataInterfaceCurve
  &NiagaraDataInterfaceGBuffer
  &NiagaraDataInterfaceCollisionQuery
  &NiagaraDataInterfaceLandscape
  &NiagaraDataInterfaceCamera
  &NiagaraDataInterfaceColorCurve
  &NiagaraDataInterfaceSpriteRendererInfo
 ]
 usedDynamicInputs [/Niagara/DynamicInputs/Helpers/SimulationPosition]
 usedModules [
  &SystemState
  &self:SystemSpawnScript.SystemScriptSource.SystemScriptGraph.NiagaraNodeAssignment_3.SetVariables_#1
  &self:SystemSpawnScript.SystemScriptSource.SystemScriptGraph.NiagaraNodeAssignment_1.SetVariables_#2
 ]
}
emitters {
 Glints {
  simTarget=GPUComputeSim
  emitterUpdateScript [
   &EmitterState()
   &SpawnRate(SpawnRate=5000, Use Spawn Probability=false) {module=SpawnRate001}
   &SpawnRate(
    SpawnRate=&Multiply_Float(
     A=2500
     B=&FloatFromCurve(
      FloatCurve=DataInterface() {propertyValues {Curve {keys [
          {interpMode=RCIM_Cubic, tangentMode=RCTM_User, time=0, value=1}
          <at 0.1 0.2>
          {interpMode=RCIM_Cubic, tangentMode=RCTM_User, time=0.25, value=0.25}
          {interpMode=RCIM_Cubic, tangentMode=RCTM_User, time=0.4199, value=0.9966}
          {interpMode=RCIM_Cubic, tangentMode=RCTM_User, time=0.5, value=1}
          {interpMode=RCIM_Cubic, tangentMode=RCTM_User, time=0.5967, value=1.0046}
          <at 0.7666 0.25>
          {interpMode=RCIM_Cubic, tangentMode=RCTM_User, time=0.95, value=0.25}
          {interpMode=RCIM_Cubic, tangentMode=RCTM_User, time=1, value=1}
         ]}}}
      CurveIndex=Linked() {linkedVariable {name=NPC.FortniteNPC.TimeOfDay_0-1, type=NiagaraFloat}}
     )
    )
    Use Spawn Probability=false
   ) {enabled=false}
  ]
  particleSpawnScript [
   /Niagara/Modules/Spawn/Initialization/InitializeParticle(
    Lifetime=&UniformRangedFloat(Minimum=2, Maximum=3)
    Module.Write Color=true
    Module.Write Lifetime=true
    Module.Write Mass=true
    Module.Write Material Random=true
    Module.Write Position=false
    Module.Write Scale=false
    Module.Write SpriteRotation=true
    Module.Write SpriteSize=true
    Sprite Rotation=&UniformRangedFloat(Minimum=0, Maximum=360)
    Sprite Size=&Vector2DFromFloat(Value=&UniformRangedFloat(Minimum=6, Maximum=7))
   ) {enabled=false}
   /Niagara/Modules/Spawn/Initialization/V2/InitializeParticle(
    Color=&ColorFromCurve(
     ColorCurve=DataInterface() {propertyValues {
       AlphaCurve {keys [<at 0 1>,<at 1 1>]}
       BlueCurve {keys [
         {arriveTangent=-1.3295, time=0.084, value=1}
         <at 0.3376 0.3613>
         {arriveTangent=0.6794, leaveTangent=0.6794, time=0.5764, value=0.5647}
         <at 1 0.2874>
        ]}
       GreenCurve {keys [
         {arriveTangent=0.1342, time=0.084, value=0.7445}
         <at 0.3376 0.0369>
         {arriveTangent=-1.1197, leaveTangent=-1.1197, time=0.5764, value=0.2664}
         <at 1 0.0953>
        ]}
       RedCurve {keys [
         {arriveTangent=0.0004, time=0.084, value=0}
         <at 0.3376 0.6654>
         {arriveTangent=-0.5897, leaveTangent=-0.5897, time=0.5764, value=0.1559}
         <at 1 0.0203>
        ]}
      }}
     CurveIndex=&RandomRangeFloat(Minimum=0, Maximum=1)
    )
    Lifetime=&UniformRangedFloat(Minimum=2, Maximum=3)
    Sprite Size Mode=::NewEnumerator1
    Uniform Sprite Size Max=100
    Uniform Sprite Size Min=50
    UsePositionOffset=false
    Write Material Random=true
   ) {module=InitializeParticle001}
   &ShapeLocation(MaskSpawnGroup=false, Offset=<0 0 5000>, Offset Mode=::NewEnumerator0, Override Seed=false, Sphere Radius=10000, UseOwnerScale=true)
   /Niagara/Modules/Update/Utility/SampleGBuffer(
    Buffer To Query:ENiagaraGBufferList_v2=::NewEnumerator8
    Kill Beyond Depth=true
    Max Z Depth=15000
    Snap To Depth:ENiagaraGBufferSnapToDepth=::NewEnumerator1
    Write To Color=false
   )
   /Game/Effects/Niagara/Modules/Spawn/FN_Spawn_GBuffer_Color_WithinRangeV2(HueMax=125, HueMin=100)
   &KillParticlesInVolume(
    Invert Volume=true
    Sphere Radius=25000
    Volume Origin:Vector3f=&AddVectorToPosition(Position=<-75208 -51836 4000>, Vector=<0 0 0>)
   )
   &self:ViewMask_01(Border Percent Max=25, Border Percent Min=15) {module=ViewMask_01}
   &KillParticles(Kill Particles=&SetBoolByFloatComparison(A=0.1, B=$Output.ViewMask_01.ViewMask)) {module=KillParticles002}
   &KillParticlesInVolume(
    Invert Volume=true
    Sphere Radius=8000
    Volume Origin:Vector3f=&AddVectorToPosition(Position=<-75208 -51836 3084>, Vector=<0 0 0>)
   ) {module=KillParticlesInVolume001}
  ]
  particleUpdateScript [
   &ParticleState() {module=UpdateAge}
   &self:ViewMask(Border Percent Max=15, Border Percent Min=5) {module=ViewMask}
   &ScaleSpriteSize(
    Uniform Curve Scale=&Multiply_Float(
     A=&Lerp_Float(A=0, B=1, Alpha=$Output.ViewMask.ViewMask)
     B=/Niagara/DynamicInputs/Camera/CameraFade(
      Camera Query:NiagaraDataInterfaceCamera=DataInterface()
      Fade Offset:NiagaraFloat=2500
      Fade Distance:NiagaraFloat=2500
      Use Owner Position:NiagaraBool=true
     )
    )
    Uniform Curve Sprite Scale=DataInterface() {propertyValues {Curve {
       defaultValue=0
       keys [
        {arriveTangent=2.0633, interpMode=RCIM_Cubic, leaveTangent=2.0633, tangentMode=RCTM_User, time=0, value=0.5}
        <at 0.9171 1.0194 RCIM_Cubic>
        {arriveTangent=0.0224, interpMode=RCIM_Cubic, leaveTangent=0.0224, tangentMode=RCTM_User, time=1, value=0.5}
       ]
      }}}
   )
   &ScaleColor(
    Scale Alpha=&FloatFromCurve(
     FloatCurve=DataInterface() {propertyValues {Curve {keys [
         <at -0.0007 -0.0138 RCIM_Cubic>
         {arriveTangent=2.6721, interpMode=RCIM_Cubic, leaveTangent=2.6721, tangentMode=RCTM_User, time=0.3127, value=0.3952}
         {arriveTangent=-0.5712, interpMode=RCIM_Cubic, leaveTangent=-0.5712, tangentMode=RCTM_User, time=0.5424, value=1.0086}
         <at 1.0011 0.002 RCIM_Cubic>
        ]}}}
     Scale Curve=&Multiply_Float(
      A=1
      B=&FloatFromCurve(FloatCurve=DataInterface() {propertyValues {Curve {keys [<at 0 1>,<at 1.5 0.25>]}}}, CurveIndex=$Particles.RegionMask.Float, Scale Curve=0.5)
     )
    )
    Scale RGB=<25 25 25>
    ScaleA=true
    ScaleRGB=true
   )
   &SolveForcesAndVelocity(Clamp Velocity=true, Limit Acceleration=false)
   &SpriteRotationRate(Rotation Rate=&UniformRangedFloat(Minimum=-15, Maximum=25))
   /SetVariables(Particles.RandomGlintVector:Vector3f=&RotateVectorByQuaternion(
     Vector To Rotate=&RandomVector(Vector Scale=1)
     Quaternion=&MakeQuatFromAxisAngle(
      Axis=&RandomVector(Vector Scale=1)
      AngleInDegrees=&Multiply_Float(A=$Particles.NormalizedAge, B=&RandomRangeFloat(Minimum=2000, Maximum=3000))
     )
    )) {isSetParameters=true, module=SetVariables_#3}
   &AlignSpriteToMeshOrientation(SetAlignment=true, SetFacing=true)
   &CameraOffset(Camera Offset Amount=25)
   /SetVariables(
    Particles.Linear Color=&Multiply_LinearColorByFloat(Linear Color=$Particles.Color, Float=25)
    Particles.Vector 2D:Vector2f=&Multiply_Vector2D_ByFloat(Float=0.02, Vector2D=$Particles.SpriteSize)
   ) {isSetParameters=true}
  ]
  rendererClasses [&NiagaraSpriteRendererProperties]
  renderers [
   {rendererClass=&NiagaraSpriteRendererProperties}
   {rendererClass=&NiagaraSpriteRendererProperties, rendererIndex=1}
  ]
 }
 SnowSmoke {
  simTarget=GPUComputeSim
  emitterUpdateScript [
   &EmitterState(
    Life Cycle Mode=::NewEnumerator1
    Loop Delay=&RandomRangeFloat(Minimum=1, Maximum=2)
    Loop Duration=&RandomRangeFloat(Minimum=2, Maximum=3)
    Recalculate Duration Each Loop=true
    UseLoopDelay=true
   )
   &SpawnRate(
    SpawnRate=&FloatFromCurve(FloatCurve=$System.SpawnCurve, CurveIndex=$Emitter.NormalizedLoopAge, Scale Curve=1000)
    Use Spawn Probability=false
   ) {module=SpawnRate001}
   &SpawnRate(
    SpawnRate=&Multiply_Float(
     A=2500
     B=&FloatFromCurve(
      FloatCurve=DataInterface() {propertyValues {Curve {keys [
          {interpMode=RCIM_Cubic, tangentMode=RCTM_User, time=0, value=1}
          <at 0.1 0.25>
          {interpMode=RCIM_Cubic, tangentMode=RCTM_User, time=0.25, value=0.25}
          {interpMode=RCIM_Cubic, tangentMode=RCTM_User, time=0.4199, value=0.9966}
          {interpMode=RCIM_Cubic, tangentMode=RCTM_User, time=0.5, value=1}
          {interpMode=RCIM_Cubic, tangentMode=RCTM_User, time=0.5967, value=1.0046}
          <at 0.7666 0.25>
          {interpMode=RCIM_Cubic, tangentMode=RCTM_User, time=0.95, value=0.25}
          {interpMode=RCIM_Cubic, tangentMode=RCTM_User, time=1, value=1}
         ]}}}
      CurveIndex=Linked() {linkedVariable {name=NPC.FortniteNPC.TimeOfDay_0-1, type=NiagaraFloat}}
     )
    )
    Use Spawn Probability=false
   ) {enabled=false}
  ]
  particleSpawnScript [
   /Niagara/Modules/Spawn/Initialization/InitializeParticle(
    Lifetime=&UniformRangedFloat(Minimum=2, Maximum=3)
    Module.Write Color=true
    Module.Write Lifetime=true
    Module.Write Mass=true
    Module.Write Material Random=true
    Module.Write Position=false
    Module.Write Scale=false
    Module.Write SpriteRotation=true
    Module.Write SpriteSize=true
    Sprite Rotation=&UniformRangedFloat(Minimum=0, Maximum=360)
    Sprite Size=&Vector2DFromFloat(Value=&UniformRangedFloat(Minimum=5, Maximum=7))
   ) {enabled=false}
   /Niagara/Modules/Spawn/Initialization/V2/InitializeParticle(
    Lifetime Max=3
    Lifetime Min=2
    Lifetime Mode=::NewEnumerator1
    Sprite Rotation Mode=::NewEnumerator2
    Sprite Size Mode=::NewEnumerator1
    Sprite UV Mode=::NewEnumerator1
    Uniform Sprite Size Max=1500
    Uniform Sprite Size Min=800
    UsePositionOffset=false
    Write Material Random=true
   ) {module=InitializeParticle001}
   &ShapeLocation(
    Hemisphere Distribution=<1 0.5>
    MaskSpawnGroup=false
    Offset=<0 0 2500>
    Offset Mode=::NewEnumerator0
    Override Seed=false
    Sphere Radius=10000
    UseOwnerScale=true
   )
   &SampleGBuffer(
    Buffer To Query:ENiagaraGBufferList_v2=::NewEnumerator8
    Kill Beyond Depth=true
    Max Z Depth=10000
    Snap To Depth:ENiagaraGBufferSnapToDepth=::NewEnumerator1
    Write To Color=false
   )
   &FN_Spawn_GBuffer_Color_WithinRangeV2(HueMax=150, HueMin=100)
   &AddVelocity(Velocity=&UniformRangedVector(Minimum=<600 -256 100>, Maximum=<1200 256 150>)) {enabled=false}
   &AddVelocityInCone(Cone Angle=15, Cone Axis Coordinate Space=::World, Module.Use Velocity Falloff On Cone Axis=false, Velocity Strength=250) {module=AddVelocityInCone001}
   &KillParticlesInVolume(Sphere Radius=3000)
   &self:ViewMask_01(Border Percent Max=25, Border Percent Min=15) {module=ViewMask_01}
   &KillParticles(Kill Particles=&SetBoolByFloatComparison(A=0.2, B=$Output.ViewMask_01.ViewMask)) {module=KillParticles002}
  ]
  particleUpdateScript [
   &ParticleState() {module=UpdateAge}
   &ScaleSpriteSize(
    Uniform Curve Scale=&Multiply_Float(
     A=1
     B=&CameraFade(
      Camera Query:NiagaraDataInterfaceCamera=DataInterface()
      Fade Offset:NiagaraFloat=1000
      Fade Distance:NiagaraFloat=2500
      Use Owner Position:NiagaraBool=true
     )
    )
    Uniform Curve Sprite Scale=DataInterface() {propertyValues {Curve {
       defaultValue=0
       keys [
        {arriveTangent=4.1785, interpMode=RCIM_Cubic, leaveTangent=4.1785, tangentMode=RCTM_User, time=0, value=0.5}
        {arriveTangent=0.0224, interpMode=RCIM_Cubic, leaveTangent=0.0224, tangentMode=RCTM_User, time=0.5, value=1}
       ]
      }}}
   )
   &RegionMask(
    Inner Float=2
    Offset:Vector3f=<-2500 0 0>
    Offset Space:ENiagaraCoordinateSpace=::World
    Outer Float=1
    Region Origin:Vector3f=&AddVectorToPosition(Position=<-48556 -33376 632>, Vector=<0 0 0>)
    Region Radius=35000
    UseOffset=true
    UseRotation=false
   )
   &ScaleColor(
    Scale Alpha=&FloatFromCurve(
     FloatCurve=DataInterface() {propertyValues {Curve {keys [
         <at -0.0007 -0.0138 RCIM_Cubic>
         {arriveTangent=-0.5712, interpMode=RCIM_Cubic, leaveTangent=-0.5712, tangentMode=RCTM_User, time=0.5424, value=1.0086}
         <at 1.0011 0.002 RCIM_Cubic>
        ]}}}
     Scale Curve=&Multiply_Float(
      A=0.15
      B=&CameraFade(
       Camera Query:NiagaraDataInterfaceCamera=DataInterface()
       Fade Offset:NiagaraFloat=2500
       Fade Distance:NiagaraFloat=5000
       Use Owner Position:NiagaraBool=true
      )
     )
    )
    Scale RGB=&Multiply_VectorByFloat(
     Vector=&MakeVectorFromLinearColorRGB(Color=&ColorFromCurve(
       ColorCurve=DataInterface() {propertyValues {
         AlphaCurve {keys [{interpMode=RCIM_Cubic, tangentMode=RCTM_User, time=0, value=1}, {interpMode=RCIM_Cubic, tangentMode=RCTM_User, time=1, value=1}]}
         BlueCurve {keys [
           {interpMode=RCIM_Cubic, leaveTangentWeight=0.0473, tangentMode=RCTM_User, time=0, value=0.092}
           {
            arriveTangent=-0.0529
            arriveTangentWeight=0.0474
            interpMode=RCIM_Cubic
            leaveTangent=-0.0529
            leaveTangentWeight=0.047
            tangentMode=RCTM_User
            time=0
            value=0.1781
           }
           {
            arriveTangent=-0.0887
            arriveTangentWeight=0.0471
            interpMode=RCIM_Cubic
            leaveTangent=-0.0887
            leaveTangentWeight=0.0072
            tangentMode=RCTM_User
            time=0.1409
            value=0.1539
           }
           {
            arriveTangent=-1.1304
            arriveTangentWeight=0.0224
            interpMode=RCIM_Cubic
            leaveTangent=-1.1304
            leaveTangentWeight=0.0193
            tangentMode=RCTM_User
            time=0.1623
            value=0.1367
           }
           {
            arriveTangent=-0.664
            arriveTangentWeight=0.0074
            interpMode=RCIM_Cubic
            leaveTangent=-0.664
            leaveTangentWeight=0.0094
            tangentMode=RCTM_User
            time=0.1809
            value=0.1087
           }
           {
            arriveTangent=0.6098
            arriveTangentWeight=0.0092
            interpMode=RCIM_Cubic
            leaveTangent=0.6098
            leaveTangentWeight=0.0028
            tangentMode=RCTM_User
            time=0.2044
            value=0.1087
           }
           {
            arriveTangent=2.3315
            arriveTangentWeight=0.0061
            interpMode=RCIM_Cubic
            leaveTangent=2.3315
            leaveTangentWeight=0.0366
            tangentMode=RCTM_User
            time=0.2116
            value=0.1275
           }
           {
            arriveTangent=1.1205
            arriveTangentWeight=0.0217
            interpMode=RCIM_Cubic
            leaveTangent=1.1205
            leaveTangentWeight=0.0372
            tangentMode=RCTM_User
            time=0.2549
            value=0.2264
           }
           {
            arriveTangent=0.2237
            arriveTangentWeight=0.0253
            interpMode=RCIM_Cubic
            leaveTangent=0.2237
            leaveTangentWeight=0.0648
            tangentMode=RCTM_User
            time=0.3291
            value=0.2591
           }
           {
            arriveTangent=0.1845
            arriveTangentWeight=0.0643
            interpMode=RCIM_Cubic
            leaveTangent=0.1845
            leaveTangentWeight=0.0381
            tangentMode=RCTM_User
            time=0.5187
            value=0.2854
           }
           {
            arriveTangent=0.1461
            arriveTangentWeight=0.0379
            interpMode=RCIM_Cubic
            leaveTangent=0.1461
            leaveTangentWeight=0.026
            tangentMode=RCTM_User
            time=0.6311
            value=0.3149
           }
           {
            arriveTangent=-0.9326
            arriveTangentWeight=0.0331
            interpMode=RCIM_Cubic
            leaveTangent=-0.9326
            leaveTangentWeight=0.0309
            tangentMode=RCTM_User
            time=0.7083
            value=0.3131
           }
           {
            arriveTangent=-1.0096
            arriveTangentWeight=0.0428
            interpMode=RCIM_Cubic
            leaveTangent=-1.0096
            leaveTangentWeight=0.0513
            tangentMode=RCTM_User
            time=0.7805
            value=0.1755
           }
           {
            arriveTangent=-0.4797
            arriveTangentWeight=0.0429
            interpMode=RCIM_Cubic
            leaveTangent=-0.4797
            leaveTangentWeight=0.0155
            tangentMode=RCTM_User
            time=0.8672
            value=0.1527
           }
           {
            arriveTangent=-0.5158
            arriveTangentWeight=0.0232
            interpMode=RCIM_Cubic
            leaveTangent=-0.5158
            leaveTangentWeight=0.0137
            tangentMode=RCTM_User
            time=0.8985
            value=0.1189
           }
           {
            arriveTangent=0.5759
            arriveTangentWeight=0.0086
            interpMode=RCIM_Cubic
            leaveTangent=0.5759
            leaveTangentWeight=0.0134
            tangentMode=RCTM_User
            time=0.917
            value=0.127
           }
           {
            arriveTangent=0.4127
            arriveTangentWeight=0.0121
            interpMode=RCIM_Cubic
            leaveTangent=0.4127
            leaveTangentWeight=0.014
            tangentMode=RCTM_User
            time=0.9458
            value=0.1461
           }
           {
            arriveTangent=0.5904
            arriveTangentWeight=0.014
            interpMode=RCIM_Cubic
            leaveTangent=0.5904
            leaveTangentWeight=0.0087
            tangentMode=RCTM_User
            time=0.9792
            value=0.1527
           }
           {arriveTangentWeight=0.0069, interpMode=RCIM_Cubic, tangentMode=RCTM_User, time=1, value=0.1781}
          ]}
         GreenCurve {keys [
           {interpMode=RCIM_Cubic, leaveTangentWeight=0.0473, tangentMode=RCTM_User, time=0, value=0.092}
           {
            arriveTangent=-0.0111
            arriveTangentWeight=0.0473
            interpMode=RCIM_Cubic
            leaveTangent=-0.0111
            leaveTangentWeight=0.047
            tangentMode=RCTM_User
            time=0
            value=0.1348
           }
           {
            arriveTangent=-0.1572
            arriveTangentWeight=0.0549
            interpMode=RCIM_Cubic
            leaveTangent=-0.1572
            leaveTangentWeight=0.0084
            tangentMode=RCTM_User
            time=0.1409
            value=0.1163
           }
           {
            arriveTangent=-0.5174
            arriveTangentWeight=0.0081
            interpMode=RCIM_Cubic
            leaveTangent=-0.5174
            leaveTangentWeight=0.007
            tangentMode=RCTM_User
            time=0.1623
            value=0.1093
           }
           {
            arriveTangent=0.2567
            arriveTangentWeight=0.0064
            interpMode=RCIM_Cubic
            leaveTangent=0.2567
            leaveTangentWeight=0.0081
            tangentMode=RCTM_User
            time=0.1809
            value=0.0985
           }
           {
            arriveTangent=1.5441
            arriveTangentWeight=0.0145
            interpMode=RCIM_Cubic
            leaveTangent=1.5441
            leaveTangentWeight=0.0044
            tangentMode=RCTM_User
            time=0.2044
            value=0.1343
           }
           {
            arriveTangent=0.6936
            arriveTangentWeight=0.0029
            interpMode=RCIM_Cubic
            leaveTangent=0.6936
            leaveTangentWeight=0.0176
            tangentMode=RCTM_User
            time=0.2116
            value=0.1461
           }
           {
            arriveTangent=0.4378
            arriveTangentWeight=0.0157
            interpMode=RCIM_Cubic
            leaveTangent=0.4378
            leaveTangentWeight=0.0266
            tangentMode=RCTM_User
            time=0.2549
            value=0.1693
           }
           {
            arriveTangent=0.1985
            arriveTangentWeight=0.0249
            interpMode=RCIM_Cubic
            leaveTangent=0.1985
            leaveTangentWeight=0.0648
            tangentMode=RCTM_User
            time=0.3281
            value=0.197
           }
           {
            arriveTangent=0.1815
            arriveTangentWeight=0.0646
            interpMode=RCIM_Cubic
            leaveTangent=0.1815
            leaveTangentWeight=0.0381
            tangentMode=RCTM_User
            time=0.5187
            value=0.2027
           }
           {
            arriveTangent=0.1637
            arriveTangentWeight=0.038
            interpMode=RCIM_Cubic
            leaveTangent=0.1637
            leaveTangentWeight=0.0261
            tangentMode=RCTM_User
            time=0.6311
            value=0.2521
           }
           {
            arriveTangent=-0.5568
            arriveTangentWeight=0.028
            interpMode=RCIM_Cubic
            leaveTangent=-0.5568
            leaveTangentWeight=0.0262
            tangentMode=RCTM_User
            time=0.7083
            value=0.2527
           }
           {
            arriveTangent=-0.4339
            arriveTangentWeight=0.0323
            interpMode=RCIM_Cubic
            leaveTangent=-0.4339
            leaveTangentWeight=0.0387
            tangentMode=RCTM_User
            time=0.7805
            value=0.1689
           }
           {
            arriveTangent=-0.2899
            arriveTangentWeight=0.039
            interpMode=RCIM_Cubic
            leaveTangent=-0.2899
            leaveTangentWeight=0.0141
            tangentMode=RCTM_User
            time=0.8672
            value=0.1838
           }
           {
            arriveTangent=-1.2376
            arriveTangentWeight=0.0302
            interpMode=RCIM_Cubic
            leaveTangent=-1.2376
            leaveTangentWeight=0.0178
            tangentMode=RCTM_User
            time=0.8985
            value=0.1347
           }
           {
            arriveTangent=0.0037
            arriveTangentWeight=0.0113
            interpMode=RCIM_Cubic
            leaveTangent=0.0037
            leaveTangentWeight=0.0176
            tangentMode=RCTM_User
            time=0.917
            value=0.1221
           }
           {
            arriveTangent=0.2047
            arriveTangentWeight=0.0134
            interpMode=RCIM_Cubic
            leaveTangent=0.2047
            leaveTangentWeight=0.0155
            tangentMode=RCTM_User
            time=0.9458
            value=0.1348
           }
           {
            arriveTangent=-0.0002
            arriveTangentWeight=0.0187
            interpMode=RCIM_Cubic
            leaveTangent=-0.0002
            leaveTangentWeight=0.0117
            tangentMode=RCTM_User
            time=0.9792
            value=0.1348
           }
           {arriveTangentWeight=0.0069, interpMode=RCIM_Cubic, tangentMode=RCTM_User, time=1, value=0.1348}
          ]}
         RedCurve {keys [
           {interpMode=RCIM_Cubic, leaveTangentWeight=0.0473, tangentMode=RCTM_User, time=0, value=0.092}
           {
            arriveTangent=0.0279
            arriveTangentWeight=0.0485
            interpMode=RCIM_Cubic
            leaveTangent=0.0279
            leaveTangentWeight=0.0482
            tangentMode=RCTM_User
            time=0
            value=0.1119
           }
           {
            arriveTangent=0.0066
            arriveTangentWeight=0.0514
            interpMode=RCIM_Cubic
            leaveTangent=0.0066
            leaveTangentWeight=0.0078
            tangentMode=RCTM_User
            time=0.1409
            value=0.0999
           }
           {
            arriveTangent=0.8688
            arriveTangentWeight=0.0095
            interpMode=RCIM_Cubic
            leaveTangent=0.8688
            leaveTangentWeight=0.0082
            tangentMode=RCTM_User
            time=0.1623
            value=0.113
           }
           {
            arriveTangent=2.2289
            arriveTangentWeight=0.0151
            interpMode=RCIM_Cubic
            leaveTangent=2.2289
            leaveTangentWeight=0.0192
            tangentMode=RCTM_User
            time=0.1809
            value=0.1422
           }
           {
            arriveTangent=3.0926
            arriveTangentWeight=0.0256
            interpMode=RCIM_Cubic
            leaveTangent=3.0926
            leaveTangentWeight=0.0078
            tangentMode=RCTM_User
            time=0.2044
            value=0.2412
           }
           {
            arriveTangent=-1.4403
            arriveTangentWeight=0.0042
            interpMode=RCIM_Cubic
            leaveTangent=-1.4403
            leaveTangentWeight=0.0253
            tangentMode=RCTM_User
            time=0.2116
            value=0.2374
           }
           {
            arriveTangent=-0.4667
            arriveTangentWeight=0.0159
            interpMode=RCIM_Cubic
            leaveTangent=-0.4667
            leaveTangentWeight=0.0269
            tangentMode=RCTM_User
            time=0.2549
            value=0.1685
           }
           {
            arriveTangent=0.1084
            arriveTangentWeight=0.0245
            interpMode=RCIM_Cubic
            leaveTangent=0.1084
            leaveTangentWeight=0.0639
            tangentMode=RCTM_User
            time=0.3281
            value=0.1831
           }
           {
            arriveTangent=0.1435
            arriveTangentWeight=0.0642
            interpMode=RCIM_Cubic
            leaveTangent=0.1435
            leaveTangentWeight=0.0378
            tangentMode=RCTM_User
            time=0.5187
            value=0.1724
           }
           {
            arriveTangent=0.1553
            arriveTangentWeight=0.0379
            interpMode=RCIM_Cubic
            leaveTangent=0.1553
            leaveTangentWeight=0.026
            tangentMode=RCTM_User
            time=0.6311
            value=0.2266
           }
           {
            arriveTangent=-0.4817
            arriveTangentWeight=0.0273
            interpMode=RCIM_Cubic
            leaveTangent=-0.4817
            leaveTangentWeight=0.0255
            tangentMode=RCTM_User
            time=0.7083
            value=0.2266
           }
           {
            arriveTangent=-0.2068
            arriveTangentWeight=0.0242
            interpMode=RCIM_Cubic
            leaveTangent=-0.2068
            leaveTangentWeight=0.029
            tangentMode=RCTM_User
            time=0.7805
            value=0.1546
           }
           {
            arriveTangent=0.4933
            arriveTangentWeight=0.0291
            interpMode=RCIM_Cubic
            leaveTangent=0.4933
            leaveTangentWeight=0.0105
            tangentMode=RCTM_User
            time=0.8672
            value=0.1937
           }
           {
            arriveTangent=-0.9567
            arriveTangentWeight=0.0274
            interpMode=RCIM_Cubic
            leaveTangent=-0.9567
            leaveTangentWeight=0.0161
            tangentMode=RCTM_User
            time=0.8985
            value=0.2128
           }
           {
            arriveTangent=-1.757
            arriveTangentWeight=0.0213
            interpMode=RCIM_Cubic
            leaveTangent=-1.757
            leaveTangentWeight=0.0331
            tangentMode=RCTM_User
            time=0.917
            value=0.1461
           }
           {
            arriveTangent=-0.5697
            arriveTangentWeight=0.0193
            interpMode=RCIM_Cubic
            leaveTangent=-0.5697
            leaveTangentWeight=0.0224
            tangentMode=RCTM_User
            time=0.9458
            value=0.1298
           }
           {
            arriveTangent=-0.3306
            arriveTangentWeight=0.0218
            interpMode=RCIM_Cubic
            leaveTangent=-0.3306
            leaveTangentWeight=0.0136
            tangentMode=RCTM_User
            time=0.9792
            value=0.1106
           }
           {arriveTangentWeight=0.0069, interpMode=RCIM_Cubic, tangentMode=RCTM_User, time=1, value=0.1119}
          ]}
        }}
       CurveIndex=Linked() {linkedVariable {name=NPC.FortniteNPC.TimeOfDay_0-1, type=NiagaraFloat}}
      ))
     Float=&Multiply_Float(A=1, B=$Particles.RegionMask.Float)
    )
    ScaleA=true
    ScaleRGB=true
   )
   &WindForce(
    Scale Min / Max:Vector2f=<5 1>
    Speed Min / Max:Vector2f=<50 250>
    Turbulence Frequency Mode:ENiagaraWindTurbulenceFrequencyMode=::NewEnumerator1
    Turbulence Scale Mode:ENiagaraWindTurbulenceContributionMode=::NewEnumerator1
    Wind Speed=<500 0 0>
   )
   &Drag(Drag=0.25, Rotational Drag=0.25, Use Linear Drag=true, Use Rotational Drag=true)
   &GravityForce(Gravity=<0 0 -250>)
   &SolveForcesAndVelocity(Clamp Velocity=false, Limit Acceleration=false)
   &SpriteRotationRate(Rotation Rate=&UniformRangedFloat(Minimum=-15, Maximum=25)) {enabled=false}
   &ScaleColor(Scale RGB=<100 1 1>, ScaleA=true, ScaleRGB=true) {enabled=false, module=ScaleColor001}
   &SubUVAnimation(
    Frames Per Second=32
    Loop Range Boundary:ENiagara_WrapClamp=::NewEnumerator0
    Playback Mode:ENiagara_FPSPlayrate=::NewEnumerator1
    Playback Scale Curve=DataInterface() {propertyValues {Curve {defaultValue=1, keys [{tangentMode=RCTM_User, time=0, value=0}, {tangentMode=RCTM_User, time=1, value=1}]}}}
    "Sprite Renderer "=DataInterface() {propertyValues {SpriteRenderer=&self:SnowSmoke_0.Renderer}}
    SubUV Animation Mode:ENiagara_SubUVLookupModeV2=::NewEnumerator3
    UseEndFrame=false
    UsePlayRateScale=true
    UseStartFrame=false
   )
   &DynamicMaterialParameters(
    Index 0 Param 1=500
    Module.Index0Float0Write=true
    Module.Index0Float1Write=true
    Module.Index0Float2Write=true
    Module.Index0Float3Write=true
    Module.Index1Float0Write=true
    Module.Index1Float1Write=true
    Module.Index1Float2Write=true
    Module.Index1Float3Write=true
    Write Parameter Index 1=::NewEnumerator1
   )
  ]
  rendererClasses [&NiagaraSpriteRendererProperties]
  renderers [
   {rendererClass=&NiagaraSpriteRendererProperties}
   {rendererClass=&NiagaraSpriteRendererProperties, rendererIndex=1}
  ]
 }
 SnowflakesAir {
  simTarget=GPUComputeSim
  emitterUpdateScript [
   &EmitterState()
   &SpawnRate(
    SpawnRate=&FloatFromCurve(FloatCurve=$System.SpawnCurve, CurveIndex=$Emitter.NormalizedLoopAge, Scale Curve=15000)
    Use Spawn Probability=false
   ) {module=SpawnRate001}
   &SpawnRate(
    SpawnRate=&Multiply_Float(
     A=2500
     B=&FloatFromCurve(
      FloatCurve=DataInterface() {propertyValues {Curve {keys [
          {interpMode=RCIM_Cubic, tangentMode=RCTM_User, time=0, value=1}
          <at 0.1 0.2>
          {interpMode=RCIM_Cubic, tangentMode=RCTM_User, time=0.25, value=0.25}
          {interpMode=RCIM_Cubic, tangentMode=RCTM_User, time=0.4199, value=0.9966}
          {interpMode=RCIM_Cubic, tangentMode=RCTM_User, time=0.5, value=1}
          {interpMode=RCIM_Cubic, tangentMode=RCTM_User, time=0.5967, value=1.0046}
          <at 0.7666 0.25>
          {interpMode=RCIM_Cubic, tangentMode=RCTM_User, time=0.95, value=0.25}
          {interpMode=RCIM_Cubic, tangentMode=RCTM_User, time=1, value=1}
         ]}}}
      CurveIndex=Linked() {linkedVariable {name=NPC.FortniteNPC.TimeOfDay_0-1, type=NiagaraFloat}}
     )
    )
    Use Spawn Probability=false
   ) {enabled=false}
  ]
  particleSpawnScript [
   /Niagara/Modules/Spawn/Initialization/InitializeParticle(
    Lifetime=&UniformRangedFloat(Minimum=2, Maximum=3)
    Module.Write Color=true
    Module.Write Lifetime=true
    Module.Write Mass=true
    Module.Write Material Random=true
    Module.Write Position=false
    Module.Write Scale=false
    Module.Write SpriteRotation=true
    Module.Write SpriteSize=true
    Sprite Rotation=&UniformRangedFloat(Minimum=0, Maximum=360)
    Sprite Size=&Vector2DFromFloat(Value=&UniformRangedFloat(Minimum=6, Maximum=7))
   ) {enabled=false}
   /Niagara/Modules/Spawn/Initialization/V2/InitializeParticle(
    Lifetime=&UniformRangedFloat(Minimum=2, Maximum=3)
    Sprite Size Mode=::NewEnumerator1
    Uniform Sprite Size Min=8
    UsePositionOffset=false
    Write Material Random=true
   ) {module=InitializeParticle001}
   &ShapeLocation(MaskSpawnGroup=false, Offset=<0 0 5000>, Offset Mode=::NewEnumerator0, Override Seed=false, Sphere Radius=10000, UseOwnerScale=true)
   &SampleGBuffer(
    Buffer To Query:ENiagaraGBufferList_v2=::NewEnumerator8
    Kill Beyond Depth=true
    Max Z Depth=10000
    Snap To Depth:ENiagaraGBufferSnapToDepth=::NewEnumerator1
    Write To Color=false
   )
   &FN_Spawn_GBuffer_Color_WithinRangeV2(HueMax=150, HueMin=100)
   &AddVelocityInCone(Cone Angle=1, Cone Axis=<1 0 -1>, Cone Axis Coordinate Space=::World, Module.Use Velocity Falloff On Cone Axis=false, Velocity Strength=150)
   /Niagara/Modules/Update/Forces/CurlNoiseForce(
    Curl Noise Cone Mask Angle=120
    Curl Noise Cone Mask Falloff Angle=100
    Mask Curl Noise=true
    Noise Frequency=250
    Noise Strength=800
    Pan Noise=true
    Pan Noise Field=<0 1 2>
   )
   &VortexForce(Origin Pull Amount=300, UseOriginOffset=true, Vortex Axis=<1 0 0>, Vortex Force Amount=300, Vortex Origin Offset:Vector3f=<0 0 -5000>)
   &ApplyInitialForces()
   &KillParticlesInVolume(Sphere Radius=5000)
   &self:ViewMask_01(Border Percent Max=25, Border Percent Min=15) {module=ViewMask_01}
   &KillParticles(Kill Particles=&SetBoolByFloatComparison(A=0.2, B=$Output.ViewMask_01.ViewMask)) {module=KillParticles002}
  ]
  particleUpdateScript [
   &ParticleState() {module=UpdateAge}
   &self:ViewMask(Border Percent Max=15, Border Percent Min=5) {module=ViewMask}
   &RegionMask(
    Inner Float=1.5
    Offset Space:ENiagaraCoordinateSpace=::World
    Outer Float=1
    Region Origin:Vector3f=&AddVectorToPosition(Position=<-48556 -33376 632>, Vector=<0 0 0>)
    Region Radius=35000
    UseOffset=true
    UseRotation=false
   )
   &ScaleSpriteSize(
    Uniform Curve Scale=&Multiply_Float(
     A=&Multiply_Float(
      A=&Multiply_Float(
       A=&Multiply_Float(A=1, B=$Particles.RegionMask.Float)
       B=&FloatFromCurve(
        FloatCurve=DataInterface() {propertyValues {Curve {keys [<at 0 1>,<at 0.25 0.5>,<at 0.808 0.5>,<at 1 1>]}}}
        CurveIndex=Linked() {linkedVariable {name=NPC.FortniteNPC.TimeOfDay_0-1, type=NiagaraFloat}}
       )
      )
      B=$System.Sys_Z
     )
     B=&Lerp_Float(A=0, B=1, Alpha=$Output.ViewMask.ViewMask)
    )
    Uniform Curve Sprite Scale=DataInterface() {propertyValues {Curve {
       defaultValue=0
       keys [
        {arriveTangent=2.0633, interpMode=RCIM_Cubic, leaveTangent=2.0633, tangentMode=RCTM_User, time=0, value=0.5}
        <at 0.9171 1.0194 RCIM_Cubic>
        {arriveTangent=0.0224, interpMode=RCIM_Cubic, leaveTangent=0.0224, tangentMode=RCTM_User, time=1, value=0.5}
       ]
      }}}
   )
   &ScaleColor(
    Scale Alpha=&FloatFromCurve(
     FloatCurve=DataInterface() {propertyValues {Curve {keys [
         <at -0.0007 -0.0138 RCIM_Cubic>
         {arriveTangent=2.6721, interpMode=RCIM_Cubic, leaveTangent=2.6721, tangentMode=RCTM_User, time=0.3127, value=0.3952}
         {arriveTangent=-0.5712, interpMode=RCIM_Cubic, leaveTangent=-0.5712, tangentMode=RCTM_User, time=0.5424, value=1.0086}
         <at 1.0011 0.002 RCIM_Cubic>
        ]}}}
     Scale Curve=&Multiply_Float(
      A=1
      B=&FloatFromCurve(FloatCurve=DataInterface() {propertyValues {Curve {keys [<at 0 1>,<at 1.5 0.25>]}}}, CurveIndex=$Particles.RegionMask.Float, Scale Curve=0.5)
     )
    )
    ScaleA=true
    ScaleRGB=true
   )
   &WindForce(
    Scale Min / Max:Vector2f=<5 1>
    Speed Min / Max:Vector2f=<50 250>
    Turbulence Frequency Mode:ENiagaraWindTurbulenceFrequencyMode=::NewEnumerator1
    Turbulence Scale Mode:ENiagaraWindTurbulenceContributionMode=::NewEnumerator1
    Wind Speed=&Multiply_VectorByFloat(Vector=<250 0 0>, Float=$Particles.RegionMask.Float)
   )
   &AerodynamicDrag(Aerodynamic Drag=0.1, Aerodynamic Rotational Drag=0.1, Use Linear Drag=true, Use Rotational Drag=true)
   &SolveForcesAndVelocity(Clamp Velocity=true, Limit Acceleration=false)
   &SpriteRotationRate(Rotation Rate=&UniformRangedFloat(Minimum=-15, Maximum=25))
   /SetVariables(Particles.RandomGlintVector:Vector3f=&RotateVectorByQuaternion(
     Vector To Rotate=&RandomVector(Vector Scale=1)
     Quaternion=&MakeQuatFromAxisAngle(
      Axis=&RandomVector(Vector Scale=1)
      AngleInDegrees=&Multiply_Float(A=$Particles.NormalizedAge, B=&RandomRangeFloat(Minimum=2000, Maximum=3000))
     )
    )) {isSetParameters=true, module=SetVariables_#3}
   &DynamicMaterialParameters(
    Index 0 Param 1=&RandomRangeFloat(Minimum=0, Maximum=0)
    Index 0 Param 2=&MakeFloatFromVector(Vector=$Particles.RandomGlintVector, Channel=::NewEnumerator0)
    Index 0 Param 3=&MakeFloatFromVector(Vector=$Particles.RandomGlintVector, Channel=::NewEnumerator1)
    Index 0 Param 4=&MakeFloatFromVector(Vector=$Particles.RandomGlintVector, Channel=::NewEnumerator2)
    Module.Param0WriteEnabled=true
    Module.Param1WriteEnabled=true
    Module.Param2WriteEnabled=true
    Module.Param3WriteEnabled=true
   )
   &ScaleSpriteSizeBySpeed(Max Scale Factor=<2 3>, Min Scale Factor=<1 0.5>, Velocity Threshold=250)
   &AlignSpriteToMeshOrientation(SetAlignment=true, SetFacing=true)
   &ScaleColor(Scale RGB=<100 1 1>, ScaleA=true, ScaleRGB=true) {enabled=false, module=ScaleColor001}
  ]
  rendererClasses [&NiagaraSpriteRendererProperties]
  renderers [{rendererClass=&NiagaraSpriteRendererProperties}]
 }
}
summary {
 systemName=NS_Hera_Snow_FromTrees
 emitters [
  {emitterName=SnowflakesAir, simTarget=GPUComputeSim, rendererClasses [&NiagaraSpriteRendererProperties]}
  {emitterName=SnowSmoke, simTarget=GPUComputeSim, rendererClasses [&NiagaraSpriteRendererProperties]}
  {emitterName=Glints, simTarget=GPUComputeSim, rendererClasses [&NiagaraSpriteRendererProperties]}
 ]
 userVariables []
}
