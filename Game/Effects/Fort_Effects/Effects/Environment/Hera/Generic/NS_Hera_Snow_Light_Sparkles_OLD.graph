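# Niagara system, marked OLD, for light glinting sparkles on snow in the Hera environment.
# One GPU emitter spawns short-lived sprites at a high constant rate around the system, snaps them to scene depth via GBuffer sampling and culls by a color threshold.
# Update applies a view mask, size and alpha curves, bright HDR color and rotation, feeding two sprite renderers.
# Time-of-day spawn scaling, GBuffer hue filtering and random glint vectors are disabled.
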NiagaraSystem /Game/Effects/Fort_Effects/Effects/Environment/Hera/Generic/NS_Hera_Snow_Light_Sparkles_OLD
dependencies {
 usedDataInterfaces [&NiagaraDataInterfaceCurve, &NiagaraDataInterfaceGBuffer]
 usedDynamicInputs [
  &Vector2DFromFloat
  &MakeFloatFromVector
  /Niagara/DynamicInputs/Helpers/SimulationPosition
 ]
 usedModules [
  &SystemState
  &self:SystemSpawnScript.SystemScriptSource.SystemScriptGraph.NiagaraNodeAssignment_3.SetVariables_#1
  &self:SystemSpawnScript.SystemScriptSource.SystemScriptGraph.NiagaraNodeAssignment_1.SetVariables_#2
 ]
}
emitters {Glints {
  simTarget=GPUComputeSim
  emitterUpdateScript [
   &EmitterState()
   &SpawnRate(SpawnRate=50000, Use Spawn Probability=false) {module=SpawnRate001}
   &SpawnRate(
    SpawnRate=&Multiply_Float(
     A=2500
     B=&FloatFromCurve(
      FloatCurve=DataInterface() {propertyValues {Curve {keys [
          {interpMode=RCIM_Cubic, tangentMode=RCTM_User, time=0, value=1}
          <at 0.1 0.2>
          {interpMode=RCIM_Cubic, tangentMode=RCTM_User, time=0.25, value=0.25}
          {interpMode=RCIM_Cubic, tangentMode=RCTM_User, time=0.4199, value=0.9966}
          {interpMode=RCIM_Cubic, tangentMode=RCTM_User, time=0.5, value=1}
          {interpMode=RCIM_Cubic, tangentMode=RCTM_User, time=0.5967, value=1.0046}
          <at 0.7666 0.25>
          {interpMode=RCIM_Cubic, tangentMode=RCTM_User, time=0.95, value=0.25}
          {interpMode=RCIM_Cubic, tangentMode=RCTM_User, time=1, value=1}
         ]}}}
      CurveIndex=Linked() {linkedVariable {name=NPC.FortniteNPC.TimeOfDay_0-1, type=NiagaraFloat}}
     )
    )
    Use Spawn Probability=false
   ) {enabled=false}
  ]
  particleSpawnScript [
   &InitializeParticle(
    Lifetime=&UniformRangedFloat(Minimum=2, Maximum=3)
    Sprite Size Mode=::NewEnumerator1
    Uniform Sprite Size Max=100
    Uniform Sprite Size Min=50
    UsePositionOffset=false
    Write Material Random=true
   ) {module=InitializeParticle001}
   &ShapeLocation(MaskSpawnGroup=false, Offset Mode=::NewEnumerator0, Override Seed=false, Sphere Radius=2500, UseOwnerScale=true)
   /Niagara/Modules/Update/Utility/SampleGBuffer(
    Buffer To Query:ENiagaraGBufferList_v2=::NewEnumerator11
    Kill Beyond Depth=true
    Max Z Depth=5000
    Snap To Depth:ENiagaraGBufferSnapToDepth=::NewEnumerator1
   ) {module=SampleGBuffer001}
   /Game/Effects/Niagara/Modules/Spawn/FN_Spawn_GBuffer_Color_WithinRangeV2(HueMax=200, HueMin=100) {enabled=false}
   &self:ViewMask_01(Border Percent Max=25, Border Percent Min=15) {enabled=false, module=ViewMask_01}
   &KillParticles(Kill Particles=&SetBoolByFloatComparison(A=0.1, B=$Output.ViewMask_01.ViewMask)) {enabled=false, module=KillParticles002}
   &KillParticles(Kill Particles=&SetBoolByFloatComparison(
     A=&MakeFloatFromLinearColor(LinearColor=$Particles.Color, Channel:ENiagaraFloatFromLinearColorOptions=::NewEnumerator0)
     B=0.1
    ))
   &Color()
  ]
  particleUpdateScript [
   &ParticleState() {module=UpdateAge}
   &self:ViewMask(Border Percent Max=15, Border Percent Min=5) {module=ViewMask}
   &ScaleSpriteSize(Uniform Curve Sprite Scale=DataInterface() {propertyValues {Curve {
       defaultValue=0
       keys [
        {arriveTangent=2.0633, interpMode=RCIM_Cubic, leaveTangent=2.0633, tangentMode=RCTM_User, time=0, value=0.5}
        <at 0.9171 1.0194 RCIM_Cubic>
        {arriveTangent=0.0224, interpMode=RCIM_Cubic, leaveTangent=0.0224, tangentMode=RCTM_User, time=1, value=0.5}
       ]
      }}})
   &ScaleColor(
    Scale Alpha=&FloatFromCurve(
     FloatCurve=DataInterface() {propertyValues {Curve {keys [
         <at -0.0007 -0.0138 RCIM_Cubic>
         {arriveTangent=2.6721, interpMode=RCIM_Cubic, leaveTangent=2.6721, tangentMode=RCTM_User, time=0.3127, value=0.3952}
         {arriveTangent=-0.5712, interpMode=RCIM_Cubic, leaveTangent=-0.5712, tangentMode=RCTM_User, time=0.5424, value=1.0086}
         <at 1.0011 0.002 RCIM_Cubic>
        ]}}}
     Scale Curve=&Multiply_Float(
      A=1
      B=&FloatFromCurve(FloatCurve=DataInterface() {propertyValues {Curve {keys [<at 0 1>,<at 1.5 0.25>]}}}, CurveIndex=$Particles.RegionMask.Float, Scale Curve=0.5)
     )
    )
    Scale RGB=<25 25 25>
    ScaleA=true
    ScaleRGB=true
   )
   &SolveForcesAndVelocity(Clamp Velocity=true, Limit Acceleration=false)
   &SpriteRotationRate(Rotation Rate=&UniformRangedFloat(Minimum=-15, Maximum=25))
   /SetVariables(Particles.RandomGlintVector:Vector3f=&RotateVectorByQuaternion(
     Vector To Rotate=&RandomVector(Vector Scale=1)
     Quaternion=&MakeQuatFromAxisAngle(
      Axis=&RandomVector(Vector Scale=1)
      AngleInDegrees=&Multiply_Float(A=$Particles.NormalizedAge, B=&RandomRangeFloat(Minimum=2000, Maximum=3000))
     )
    )) {enabled=false, isSetParameters=true}
   &AlignSpriteToMeshOrientation(SetAlignment=true, SetFacing=true)
   &CameraOffset(Camera Offset Amount=25)
   /SetVariables(
    Particles.Linear Color=&Multiply_LinearColorByFloat(Linear Color=$Particles.Color, Float=25)
    Particles.Vector 2D:Vector2f=&Multiply_Vector2D_ByFloat(Float=0.02, Vector2D=$Particles.SpriteSize)
   ) {isSetParameters=true}
  ]
  rendererClasses [&NiagaraSpriteRendererProperties]
  renderers [
   {rendererClass=&NiagaraSpriteRendererProperties}
   {rendererClass=&NiagaraSpriteRendererProperties, rendererIndex=1}
  ]
 }}
summary {
 systemName=NS_Hera_Snow_Light_Sparkles_OLD
 emitters [{emitterName=Glints, simTarget=GPUComputeSim, rendererClasses [&NiagaraSpriteRendererProperties]}]
 userVariables []
}
