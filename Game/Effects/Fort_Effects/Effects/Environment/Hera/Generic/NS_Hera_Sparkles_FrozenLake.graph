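# Niagara GPU sparkle effect for the Hera frozen lake environment, filling a wide cylindrical area with drifting, twinkling sprites.
# HermesGold_Glitter spawns slow gold glitter offset to sampled landscape height and alpha-masked by a region.
# Parent spawns faster curl-noise-driven particles, which Child reads through an attribute reader to spawn smaller trailing sparkles.
# Waveforms drive twinkle, and camera fade scales glitter and Child sprites; Child's landscape and water-mask modules are disabled.
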NiagaraSystem /Game/Effects/Fort_Effects/Effects/Environment/Hera/Generic/NS_Hera_Sparkles_FrozenLake
dependencies {
 usedDataInterfaces [
  &NiagaraDataInterfaceCurve
  &NiagaraDataInterfaceCamera
  &NiagaraDataInterfaceGBuffer
  &NiagaraDataInterfaceParticleRead
  /Script/Niagara.NiagaraDataInterfaceTexture
 ]
 usedDynamicInputs []
 usedModules [
  &SystemState
  &self:SystemSpawnScript.SystemScriptSource.SystemScriptGraph.NiagaraNodeAssignment_0.SetVariables_#1
  &self:SystemSpawnScript.SystemScriptSource.SystemScriptGraph.NiagaraNodeAssignment_1.SetVariables_#2
 ]
}
emitters {
 Child {
  simTarget=GPUComputeSim
  emitterUpdateScript [
   &EmitterState()
   &SpawnParticlesFromOtherEmitter(
    Attribute Reader=DataInterface() {propertyValues {EmitterBinding {bindingMode=Other, emitterName=Parent}}}
    Maximum Age Spawn Threshold=2
    Minimum Age Spawn Threshold=0.1
    Spawn Rate Per Particle Cap=10
    Use Maximum Age=true
    Use Minimum Age=true
   )
  ]
  particleSpawnScript [
   &InitializeParticle(
    Lifetime Max=3
    Lifetime Min=2
    Lifetime Mode=::NewEnumerator1
    Mesh Scale Mode=::NewEnumerator1
    Mesh Uniform Scale Max=10
    Mesh Uniform Scale Min=5
    Sprite Size Mode=::NewEnumerator1
    Sprite UV Mode=::NewEnumerator1
    Uniform Sprite Size Max=30
    Uniform Sprite Size Min=10
    UsePositionOffset=false
    Write Material Random=true
   )
   &SampleParticlesFromOtherEmitter(
    Applied RGB Scale:Vector3f=<3 3 3>
    Applied Velocity Scale=0.1
    Apply Sampled Alpha=false
    Color Sampling=::NewEnumerator0
    Maximum Sampled Particle Age=2
    Module.MaxSampledAgeCutoff=true
    Module.MinSampledAgeCutoff=true
    Reject Particle By Sampled Age=true
    Reject Particle on Sampled Particle's First Frame=true
    Velocity Sampling=::NewEnumerator0
   )
   /SetVariables() {enabled=false, isSetParameters=true}
   /SetVariables(Particles.Position=&Add_Vector(
     A=Linked() {linkedVariable {name=Particles.Position, type=NiagaraPosition}}
     B=&MakeVector(X=0, Y=0, Z=$Particles.Terrain Offset)
    )) {enabled=false, isSetParameters=true, module=SetVariables_#3}
   &self:RainTextureMasking_04(
    AlphaThreshold=0.5
    BoxCenter:NiagaraPosition=<-81280 16256 910>
    BoxExtents:Vector3f=<97837 97837 1820>
    Position:NiagaraPosition=$Particles.Position
    RejectionTexture=DataInterface() {propertyValues {
      Texture=/Game/Effects/Fort_Effects/Effects/Environment/Hermes/HermesGold/Textures/T_HermesGold_FX_Watermask_Rejection_01
      TextureUserParameter {parameter {name=None, typeDefHandle {registeredTypeIndex=2}, varData []}}
     }}
   ) {enabled=false, module=RainTextureMasking_04}
   &AddVelocity(Velocity=<0 0 0>)
   &ShapeLocation(MaskSpawnGroup=false, Override Seed=false, Sphere Radius=8, UseOwnerScale=true)
   &KillParticlesInVolume(Sphere Radius=3000)
  ]
  particleUpdateScript [
   &self:Sample_Landscape_UPdate_03(Bias:NiagaraFloat=$Particles.Terrain Offset, New Landscape Sample:NiagaraDataInterfaceLandscape=$User.Landscape) {enabled=false, module=Sample_Landscape_UPdate_03}
   &ParticleState()
   &CurlNoiseForce(
    Curl Noise Cone Mask Angle=25
    Mask Curl Noise=true
    Noise Frequency=&RandomRangeFloat(Minimum=2, Maximum=4)
    Noise Strength=&Multiply_Float(A=&RandomRangeFloat(Minimum=25, Maximum=25), B=1)
    Pan Noise=true
   )
   &SolveForcesAndVelocity(
    Acceleration Limit=250
    Clamp Velocity=true
    Force:Vector3f=&Multiply_Vector(A=$Transient.PhysicsForce, B=<1 1 1>)
    Limit Acceleration=true
    Speed Limit=250
    Velocity:Vector3f=&Multiply_Vector(A=$Particles.Velocity, B=<1 1 1>)
   )
   &CalculateAccurateVelocity() {enabled=false}
   /Niagara/Modules/Math/AvoidDistanceFieldSurfaces_GPU(
    Avoid Nearest Surface=true
    Avoid Oncoming Surfaces=true
    Falloff Distance=1
    Max Trace Distance=1
    Nearest Surface Avoidance Strength=1
    Obstactle Avoidance Strength=1
    Trace Vector:Vector3f=&Multiply_Vector(A=$Particles.Velocity, B=<1 1 0>)
   ) {enabled=false}
   &DynamicMaterialParameters(
    Index 0 Param 1=0.2
    Index 0 Param 2=&RandomRangeFloat(Minimum=0.2, Maximum=0.4)
    Index 0 Param 3=1
    Module.Index0Param0Write=true
    Module.Index0Param1Write=true
    Module.Index0Param2Write=true
    Module.Index0Param3Write=true
   )
   &ScaleSpriteSize(
    Uniform Curve Scale=&Multiply_Float(
     A=1
     B=/Niagara/DynamicInputs/Camera/CameraFade(
      Camera Query:NiagaraDataInterfaceCamera=DataInterface()
      Fade Offset:NiagaraFloat=1250
      Fade Distance:NiagaraFloat=1250
      Use Owner Position:NiagaraBool=true
     )
    )
    Uniform Curve Sprite Scale=DataInterface() {propertyValues {Curve {keys [
        {interpMode=RCIM_Cubic, tangentMode=RCTM_User, time=0, value=0}
        {interpMode=RCIM_Cubic, tangentMode=RCTM_User, time=0.2, value=1}
        {tangentMode=RCTM_User, time=1, value=0}
       ]}}}
   )
   &CameraOffset(Camera Offset Amount=10) {module=CameraOffset001}
   &ScaleColor(
    Scale Alpha=&FloatFromCurve(FloatCurve=DataInterface() {propertyValues {Curve {keys [<at 0 0.2>,<at 0.2 0.8>,<at 1 0>,<at 1.8 -0.8 RCIM_Cubic>]}}})
    Scale RGB=&Multiply_VectorByFloat(
     Vector=<5 5 5>
     Float=&CameraFade(
      Camera Query:NiagaraDataInterfaceCamera=DataInterface()
      Fade Offset:NiagaraFloat=1250
      Fade Distance:NiagaraFloat=1250
      Use Owner Position:NiagaraBool=true
     )
    )
    ScaleA=true
    ScaleRGB=true
   )
   /SetVariables(
    Particles.Linear Color=&Multiply_LinearColorByFloat(
     Linear Color=$Particles.Color
     Float=&Waveform(
      Global Amplitude Scale=5
      "Waveform [1]"=::NewEnumerator0
      "[1] Frequency"=&RandomRangeFloat(Minimum=0.5, Maximum=2)
      "[1] Amplitude Min/Max"=<0 1>
     )
    )
    Particles.Vector 2D:Vector2f=&Multiply_Vector2D(A=$Particles.SpriteSize, B=<0.25 0.25>)
   ) {isSetParameters=true, module=SetVariables_#4}
  ]
  rendererClasses [&NiagaraSpriteRendererProperties]
  renderers [
   {rendererClass=&NiagaraSpriteRendererProperties}
   {rendererClass=&NiagaraSpriteRendererProperties, rendererIndex=1}
  ]
 }
 HermesGold_Glitter {
  simTarget=GPUComputeSim
  emitterUpdateScript [
   &EmitterState()
   &SpawnRate(SpawnRate=2500, Use Spawn Probability=false)
  ]
  particleSpawnScript [
   &InitializeParticle(
    Color=<rgba 1 0.516 0.0482 1>
    Lifetime Max=3
    Lifetime Min=2
    Lifetime Mode=::NewEnumerator1
    Mesh Scale Mode=::NewEnumerator1
    Mesh Uniform Scale Max=10
    Mesh Uniform Scale Min=5
    Sprite Size Mode=::NewEnumerator1
    Sprite UV Mode=::NewEnumerator1
    Uniform Sprite Size Max=20
    Uniform Sprite Size Min=10
    UsePositionOffset=false
    Write Material Random=true
   )
   &ShapeLocation(
    Cylinder Height=150
    Cylinder Radius=7500
    MaskSpawnGroup=false
    Override Seed=false
    Shape Primitive=::NewEnumerator1
    Surface Only=true
    Surface Only Band Thickness=0.95
    Use Endcaps In Surface Only Mode=false
    UseOwnerScale=true
   )
   /SetVariables(Particles.Position=&Add_Vector(
     A=Linked() {linkedVariable {name=Particles.Position, type=NiagaraPosition}}
     B=&MakeVector(X=0, Y=0, Z=$Particles.Terrain Offset)
    )) {isSetParameters=true, module=SetVariables_#3}
   &KillParticlesInVolume(Sphere Radius=3000)
   &AddVelocity(Velocity=<0 0 5>)
  ]
  particleUpdateScript [
   &self:Sample_Landscape_UPdate_02(Bias:NiagaraFloat=$Particles.Terrain Offset, New Landscape Sample:NiagaraDataInterfaceLandscape=$User.Landscape) {module=Sample_Landscape_UPdate_02}
   &ParticleState()
   &CurlNoiseForce(
    Curl Noise Cone Mask Angle=25
    Mask Curl Noise=true
    Noise Frequency=&RandomRangeFloat(Minimum=2, Maximum=4)
    Noise Strength=&Multiply_Float(A=&RandomRangeFloat(Minimum=250, Maximum=250), B=1)
    Pan Noise=true
   )
   &SolveForcesAndVelocity(
    Acceleration Limit=50
    Clamp Velocity=true
    Force:Vector3f=&Multiply_Vector(A=$Transient.PhysicsForce, B=<1 1 0>)
    Limit Acceleration=true
    Speed Limit=50
    Velocity:Vector3f=&Multiply_Vector(A=$Particles.Velocity, B=<1 1 1>)
   )
   &CalculateAccurateVelocity() {enabled=false}
   &AvoidDistanceFieldSurfaces_GPU(
    Avoid Nearest Surface=true
    Avoid Oncoming Surfaces=true
    Falloff Distance=1
    Max Trace Distance=1
    Nearest Surface Avoidance Strength=1
    Obstactle Avoidance Strength=1
    Trace Vector:Vector3f=&Multiply_Vector(A=$Particles.Velocity, B=<1 1 0>)
   ) {enabled=false}
   &RegionMask(
    Invert Region=true
    Offset Space:ENiagaraCoordinateSpace=::Local
    Region Dimensions:Vector3f=<15000 15000 15000>
    Region Type:ENiagaraDebugDrawShape=::NewEnumerator1
    UseOffset=true
    UseRotation=false
   )
   &DynamicMaterialParameters(
    Index 0 Param 1=0.2
    Index 0 Param 2=&RandomRangeFloat(Minimum=0.2, Maximum=0.4)
    Index 0 Param 3=1
    Module.Index0Param0Write=true
    Module.Index0Param1Write=true
    Module.Index0Param2Write=true
    Module.Index0Param3Write=true
   )
   &ScaleColor(
    Scale Alpha=&FloatFromCurve(
     FloatCurve=DataInterface() {propertyValues {Curve {keys [<at 0 0.2>,<at 0.2 0.8>,<at 1 0>]}}}
     Scale Curve=&Multiply_Float(A=5, B=$StackContext.RegionMask.Float)
    )
    ScaleA=true
    ScaleRGB=true
   )
   &ScaleSpriteSize(
    Uniform Curve Scale=&Multiply_Float(
     A=3
     B=&CameraFade(
      Camera Query:NiagaraDataInterfaceCamera=DataInterface()
      Fade Offset:NiagaraFloat=1250
      Fade Distance:NiagaraFloat=1250
      Use Owner Position:NiagaraBool=true
     )
    )
    Uniform Curve Sprite Scale=DataInterface() {propertyValues {Curve {keys [
        {interpMode=RCIM_Cubic, tangentMode=RCTM_User, time=0, value=0}
        {interpMode=RCIM_Cubic, tangentMode=RCTM_User, time=0.2, value=1}
        {tangentMode=RCTM_User, time=1, value=0}
       ]}}}
   )
   &CameraOffset(Camera Offset Amount=10) {module=CameraOffset001}
   /Niagara/Modules/Update/Utility/SampleGBuffer(
    Buffer To Query:ENiagaraGBufferList_v2=::NewEnumerator8
    Kill Beyond Depth=false
    Snap To Depth:ENiagaraGBufferSnapToDepth=::NewEnumerator1
    Write To Position=false
   )
   /SetVariables(
    Particles.Linear Color=&Multiply_LinearColorByFloat(
     Linear Color=$Particles.Color
     Float=&Waveform(
      Global Amplitude Scale=5
      "Waveform [1]"=::NewEnumerator0
      "[1] Frequency"=&RandomRangeFloat(Minimum=0.5, Maximum=2)
      "[1] Amplitude Min/Max"=<0 1>
     )
    )
    Particles.Vector 2D:Vector2f=&Multiply_Vector2D(A=$Particles.SpriteSize, B=<0.25 0.25>)
   ) {isSetParameters=true, module=SetVariables_#4}
  ]
  rendererClasses [&NiagaraSpriteRendererProperties]
  renderers [
   {rendererClass=&NiagaraSpriteRendererProperties}
   {rendererClass=&NiagaraSpriteRendererProperties, rendererIndex=1}
  ]
 }
 Parent {
  simTarget=GPUComputeSim
  emitterUpdateScript [
   &EmitterState()
   &SpawnRate(SpawnRate=500, Use Spawn Probability=false)
  ]
  particleSpawnScript [
   &InitializeParticle(
    Color=<rgba 1 0 0.1509 1>
    Lifetime Min=2
    Lifetime Mode=::NewEnumerator1
    Mesh Scale Mode=::NewEnumerator1
    Mesh Uniform Scale Max=10
    Mesh Uniform Scale Min=5
    Sprite Size Mode=::NewEnumerator1
    Sprite UV Mode=::NewEnumerator1
    Uniform Sprite Size Max=250
    Uniform Sprite Size Min=100
    UsePositionOffset=false
    Write Material Random=true
   )
   &AddVelocity(Velocity=<500 0 25>)
   &ShapeLocation(
    Cylinder Height=0
    Cylinder Radius=7500
    MaskSpawnGroup=false
    Override Seed=false
    Shape Primitive=::NewEnumerator1
    Surface Only=true
    Surface Only Band Thickness=0.95
    Use Endcaps In Surface Only Mode=false
    UseOwnerScale=true
   )
   /SetVariables(Particles.Position=&Add_Vector(
     A=Linked() {linkedVariable {name=Particles.Position, type=NiagaraPosition}}
     B=&MakeVector(X=0, Y=0, Z=$Particles.Terrain Offset)
    )) {isSetParameters=true, module=SetVariables_#3}
   &KillParticlesInVolume(Sphere Radius=3000)
  ]
  particleUpdateScript [
   &ParticleState()
   &CurlNoiseForce(
    Noise Frequency=&RandomRangeFloat(Minimum=1, Maximum=2)
    Noise Strength=&Multiply_Float(A=&RandomRangeFloat(Minimum=500, Maximum=1500), B=1)
    Pan Noise=true
   )
   &SolveForcesAndVelocity(
    Acceleration Limit=1000
    Clamp Velocity=true
    Force:Vector3f=&Multiply_Vector(A=$Transient.PhysicsForce, B=<1 1 0.2>)
    Limit Acceleration=true
    Velocity:Vector3f=&Multiply_Vector(A=$Particles.Velocity, B=<1 1 0.2>)
   )
   &CalculateAccurateVelocity() {enabled=false}
   &AvoidDistanceFieldSurfaces_GPU(
    Avoid Nearest Surface=true
    Avoid Oncoming Surfaces=true
    Falloff Distance=1
    Max Trace Distance=1
    Nearest Surface Avoidance Strength=1
    Obstactle Avoidance Strength=1
    Trace Vector:Vector3f=&Multiply_Vector(A=$Particles.Velocity, B=<1 1 0>)
   ) {enabled=false}
   &DynamicMaterialParameters(
    Index 0 Param 1=0.2
    Index 0 Param 2=&RandomRangeFloat(Minimum=0.2, Maximum=0.4)
    Index 0 Param 3=1
    Module.Index0Param0Write=true
    Module.Index0Param1Write=true
    Module.Index0Param2Write=true
    Module.Index0Param3Write=true
   )
   &ScaleColor(
    Scale Alpha=&FloatFromCurve(FloatCurve=DataInterface() {propertyValues {Curve {keys [<at 0 0.2>,<at 0.2 0.8>,<at 1 0>]}}}, Scale Curve=25)
    Scale RGB=&MakeVectorFromLinearColorRGB(Color=<rgba 1 1 1 1>)
    ScaleA=true
    ScaleRGB=true
   ) {enabled=false}
   &SampleGBuffer(
    Buffer To Query:ENiagaraGBufferList_v2=::NewEnumerator8
    Kill Beyond Depth=false
    Snap To Depth:ENiagaraGBufferSnapToDepth=::NewEnumerator1
    Write To Position=false
   )
   /SetVariables(
    Particles.Linear Color=&Multiply_LinearColorByFloat(
     Linear Color=$Particles.Color
     Float=&Waveform(
      Global Amplitude Scale=5
      "Waveform [1]"=::NewEnumerator0
      "[1] Frequency"=&RandomRangeFloat(Minimum=0.5, Maximum=2)
      "[1] Amplitude Min/Max"=<0 1>
     )
    )
    Particles.Vector 2D:Vector2f=&Multiply_Vector2D(A=$Particles.SpriteSize, B=<0.25 0.25>)
   ) {isSetParameters=true, module=SetVariables_#4}
  ]
  rendererClasses [&NiagaraSpriteRendererProperties]
  renderers [
   {rendererClass=&NiagaraSpriteRendererProperties}
   {rendererClass=&NiagaraSpriteRendererProperties, rendererIndex=1}
  ]
 }
}
summary {
 systemName=NS_Hera_Sparkles_FrozenLake
 emitters [
  {emitterName=HermesGold_Glitter, simTarget=GPUComputeSim, rendererClasses [&NiagaraSpriteRendererProperties]}
  {emitterName=Parent, simTarget=GPUComputeSim, rendererClasses [&NiagaraSpriteRendererProperties]}
  {emitterName=Child, simTarget=GPUComputeSim, rendererClasses [&NiagaraSpriteRendererProperties]}
 ]
 userVariables [
  User.BaseRadiusScale {type=NiagaraFloat, defaultValue=1, description=""}
  User.Landscape {
   type=NiagaraDataInterfaceLandscape
   description=""
   defaultValue {
    struct=NiagaraExt_VariableValue_DataInterface
    value {
     dataInterface=&self:NiagaraDataInterfaceLandscape_1
     dataInterfaceClass=&NiagaraDataInterfaceLandscape
    }
   }
  }
  User.MaxVel {type=NiagaraFloat, defaultValue=100, description=""}
  User.NumberofBeams {type=NiagaraInt32, defaultValue=32}
  User.NumberofRings {type=NiagaraInt32, defaultValue=3}
  User.RadiusSpread {type=NiagaraFloat, defaultValue=1500, description=""}
  User.RingSize {
   type=NiagaraDataInterfaceArrayFloat
   defaultValue {
    struct=NiagaraExt_VariableValue_DataInterface
    value {
     dataInterface=&self:NiagaraDataInterfaceArrayFloat_0
     dataInterfaceClass=&NiagaraDataInterfaceArrayFloat
    }
   }
  }
  User.SplashOffset {
   type=NiagaraDataInterfaceArrayFloat3
   description=""
   defaultValue {
    struct=NiagaraExt_VariableValue_DataInterface
    value {
     dataInterface=&self:NiagaraDataInterfaceArrayFloat3_0
     dataInterfaceClass=&NiagaraDataInterfaceArrayFloat3
    }
   }
  }
  User.WaterBodyIndex {type=NiagaraFloat, defaultValue=0, description=""}
  User.WaterDepth {type=NiagaraFloat, defaultValue=0, description=""}
  User.WaterPlaneLocation {type=Vector3f, description="", defaultValue {struct=Vector3f, value=<0 0 0>}}
  User.WaterPlaneNormal {type=Vector3f, description="", defaultValue {struct=Vector3f, value=<0 0 0>}}
  User.WaveAttenuationFactor {type=NiagaraFloat, defaultValue=0, description=""}
  User.WaveTime {type=NiagaraFloat, defaultValue=0, description=""}
 ]
}
